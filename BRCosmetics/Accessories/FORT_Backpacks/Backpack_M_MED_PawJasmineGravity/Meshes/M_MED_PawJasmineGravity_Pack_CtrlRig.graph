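# Control Rig for the PawJasmineGravity backpack cosmetic that corrects the pack and a camera cord bone as the pose changes.
# Local Z rotations of five spine reference bones are each remapped to up to a fifth of a weight, then summed.
# That weight blends a self-relative tilt and small offset onto the pack bone and its children.
# A spherical pose reader on dyn_camera_01, with debug draw enabled, then weights a small nudge of cord_mid.
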ControlRigBlueprint /BRCosmetics/Accessories/FORT_Backpacks/Backpack_M_MED_PawJasmineGravity/Meshes/M_MED_PawJasmineGravity_Pack_CtrlRig
graphs [
 Add_1_ContainedGraph {nodes [
   n1 = RigVMFunctionEntryNode()
   Add = &RigVMFunction_MathDoubleAdd(A=@n1.A, B=@n1.B)
   n2 = &RigVMFunction_MathDoubleAdd(A=@Add.Result, B=@n1.C)
   RigVMFunctionReturnNode(Result=&RigVMFunction_MathDoubleAdd(A=&RigVMFunction_MathDoubleAdd(A=@n2.Result, B=@n1.D).Result, B=@n1.E).Result)
  ]}
 RigVMFunctionLibrary {links [], nodes []}
 RigVMModel {nodes [
   RigUnit_BeginExecution = &RigUnit_BeginExecution()
   n1 = &RigUnit_GetTransform(Item=(Type=Bone,Name="spine_02_ref"), Space=LocalSpace, bInitial=False)
   n2 = &RigUnit_GetTransform(Item=(Type=Bone,Name="spine_03_ref"), Space=LocalSpace, bInitial=False)
   n3 = &RigUnit_GetTransform(Item=(Type=Bone,Name="spine_01_ref"), Space=LocalSpace, bInitial=False)
   n4 = &RigUnit_GetTransform(Item=(Type=Bone,Name="spine_04_ref"), Space=LocalSpace, bInitial=False)
   n5 = &RigUnit_GetTransform(Item=(Type=Bone,Name="spine_05_ref"), Space=LocalSpace, bInitial=False)
   Add = RigVMAggregateNode(
    B=&RigVMFunction_MathDoubleRemap(
     SourceMaximum=-15
     SourceMinimum=-5
     TargetMaximum=0.2
     TargetMinimum=0
     Value=&RigVMFunction_MathQuaternionToEuler(RotationOrder=YXZ, Value=@n1.Transform.Rotation).Result.Z
    ).Result
    C=&RigVMFunction_MathDoubleRemap(
     SourceMaximum=-14
     SourceMinimum=-4
     TargetMaximum=0.2
     TargetMinimum=0
     Value=&RigVMFunction_MathQuaternionToEuler(RotationOrder=YXZ, Value=@n2.Transform.Rotation).Result.Z
    ).Result
    A=&RigVMFunction_MathDoubleRemap(
     SourceMaximum=-16
     SourceMinimum=-6
     TargetMaximum=0.2
     TargetMinimum=0
     Value=&RigVMFunction_MathQuaternionToEuler(RotationOrder=YXZ, Value=@n3.Transform.Rotation).Result.Z
    ).Result
    D=&RigVMFunction_MathDoubleRemap(
     SourceMaximum=-10
     SourceMinimum=0
     TargetMaximum=0.2
     TargetMinimum=0
     Value=&RigVMFunction_MathQuaternionToEuler(RotationOrder=YXZ, Value=@n4.Transform.Rotation).Result.Z
    ).Result
    E=&RigVMFunction_MathDoubleRemap(
     SourceMaximum=-12
     SourceMinimum=-2
     TargetMaximum=0.2
     TargetMinimum=0
     Value=&RigVMFunction_MathQuaternionToEuler(RotationOrder=YXZ, Value=@n5.Transform.Rotation).Result.Z
    ).Result
   )
   /Script/ControlRig.RigUnit_SetRelativeTransformForItem(
    Child=(Type=Bone,Name="pack")
    Parent=(Type=Bone,Name="pack")
    Value=(Rotation=(X=-0.5,Y=0,Z=0,W=0.866),Translation=(X=0,Y=-1,Z=0),Scale3D=(X=1,Y=1,Z=1))
    bParentInitial=false
    bPropagateToChildren=true
    Weight=@Add.Result
    ExecutePin=@RigUnit_BeginExecution.ExecutePin
   ) {
    is=struct RigVMUnitNode
    template="Set Relative Transform::Execute(in Child,in Parent,in bParentInitial,in Value,in Weight,in bPropagateToChildren)"
   }
   > n6 = &RigUnit_SphericalPoseReader(
    ActiveRegionScaleFactors=(PositiveWidth=1.6,NegativeWidth=1,PositiveHeight=1.5,NegativeHeight=1.5)
    ActiveRegionSize=0.1
    Debug=(bDrawDebug=True,bDraw2D=False,bDrawLocalAxes=False,DebugScale=25,DebugSegments=20,DebugThickness=0.25)
    DriverAxis=(X=1,Y=0,Z=0)
    DriverItem=(Type=Bone,Name="dyn_camera_01")
    FalloffRegionScaleFactors=(PositiveWidth=1,NegativeWidth=1,PositiveHeight=2,NegativeHeight=2)
    FalloffSize=0.2
    FlipHeightScaling=False
    FlipWidthScaling=False
    OptionalParentItem=(Type=Bone,Name="None")
    RotationOffset=(X=90,Y=0,Z=-55)
   )
   > &RigUnit_SetRelativeTransformForItem(
    Child=(Type=Bone,Name="cord_mid")
    Parent=(Type=Bone,Name="cord_mid")
    Value=(Rotation=(X=0,Y=0,Z=0,W=1),Translation=(X=0,Y=-1,Z=0),Scale3D=(X=1,Y=1,Z=1))
    bParentInitial=false
    bPropagateToChildren=true
    Weight=@n6.OutputParam
   ) {template="Set Relative Transform::Execute(in Child,in Parent,in bParentInitial,in Value,in Weight,in bPropagateToChildren)"}
  ]}
]
hierarchy [
 root {loc=<0 0 0>, rot=<0 0 0 1>}
 pack {parent=root, loc=<0 -0.7316 0>, rot=<0 0 0 1>}
 dyn_zipper_top_MSTR {parent=pack, loc=<9.1653 -10.0169 -2.3498>, rot=<0.484 -0.5147 -0.3195 -0.6315>}
 dyn_zipper_top_01 {parent=dyn_zipper_top_MSTR, loc=<0 0 0>, rot=<0 0 0 1>}
 dyn_zipper_bot_MSTR {parent=pack, loc=<12.0146 -13.6053 -17.7539>, rot=<0.0508 -0.7061 -0.0514 -0.7044>}
 dyn_zipper_bot_01 {parent=dyn_zipper_bot_MSTR, loc=<0 0 0>, rot=<0 0 0 1>}
 dyn_dyn_trinket_MSTR {parent=pack, loc=<-11.9954 -13.5049 -11.3659>, rot=<0.0035 -0.6999 0.0867 -0.709>}
 dyn_dyn_trinket_01 {parent=dyn_dyn_trinket_MSTR, loc=<0 0 0>, rot=<0 0 0 1>}
 dyn_dyn_trinket_02 {parent=dyn_dyn_trinket_01, loc=<1.7626 -0.0282 -0.0745>, rot=<-0.3001 0.0426 0.0933 -0.9484>}
 dyn_dyn_trinket_03 {parent=dyn_dyn_trinket_02, loc=<0.9561 0.0651 -0.0056>, rot=<-0.0362 -0.0329 -0.0041 -0.9988>}
 dyn_backpack_r_MSTR {parent=pack, loc=<-4.4635 -15.0897 -0.2>, rot=<0.5851 -0.4398 0.5693 -0.3744>}
 dyn_backpack_r_01 {parent=dyn_backpack_r_MSTR, loc=<0 0 0>, rot=<0 0 0 1>}
 dyn_backpack_l_MSTR {parent=pack, loc=<4.5214 -13.0107 0.2361>, rot=<0.4801 -0.5041 0.3808 -0.6086>}
 dyn_backpack_l_01 {parent=dyn_backpack_l_MSTR, loc=<0 0 0>, rot=<0 0 0 1>}
 dyn_cord_l_MSTR {parent=pack, loc=<9.0524 -11.8597 0.4278>, rot=<0.5126 -0.364 -0.4757 -0.6152>}
 dyn_cord_l_01 {parent=dyn_cord_l_MSTR, loc=<0 0 0>, rot=<0 0 0 1>}
 dyn_cord_l_02 {parent=dyn_cord_l_01, loc=<4.691 0 0>, rot=<0 0 0 1>}
 dyn_strap_r_MSTR {parent=pack, loc=<-5.2507 -11.0393 -22.5128>, rot=<0.2006 -0.7097 0.197 -0.6459>}
 dyn_strap_r_01 {parent=dyn_strap_r_MSTR, loc=<0 0 0>, rot=<0 0 0 1>}
 dyn_strap_r_02 {parent=dyn_strap_r_01, loc=<2.109 -0.1469 0.0546>, rot=<-0.0168 -0.0114 -0.3323 -0.943>}
 dyn_strap_l_MSTR {parent=pack, loc=<5.2507 -11.0393 -22.5128>, rot=<0.197 -0.6459 0.2006 -0.7097>}
 dyn_strap_l_01 {parent=dyn_strap_l_MSTR, loc=<0 0 0>, rot=<0 0 0 1>}
 dyn_strap_l_02 {parent=dyn_strap_l_01, loc=<2.109 -0.1468 -0.0547>, rot=<0.0168 0.0114 -0.3323 -0.943>}
 dyn_cord_r_f_MSTR {parent=pack, loc=<-8.6746 -12.6403 0.5577>, rot=<-0.435 0.5821 0.6102 0.3157>}
 dyn_cord_r_f_01 {parent=dyn_cord_r_f_MSTR, loc=<0 0 0>, rot=<0 0 0 1>}
 dyn_cord_r_f_02 {parent=dyn_cord_r_f_01, loc=<3.8381 0 0>, rot=<0 0 0.1372 -0.9905>}
 dyn_cord_r_f_03 {parent=dyn_cord_r_f_02, loc=<4.7617 0 0>, rot=<0 0 0 1>}
 dyn_cord_r_b_MSTR {parent=pack, loc=<-8.8276 -11.5897 -0.469>, rot=<0.3513 -0.6239 -0.5828 -0.3842>}
 dyn_cord_r_b_01 {parent=dyn_cord_r_b_MSTR, loc=<0 0 0>, rot=<0 0 0 1>}
 dyn_cord_r_b_02 {parent=dyn_cord_r_b_01, loc=<3.1336 0 0>, rot=<0 0 0.069 -0.9976>}
 dyn_display_MSTR {parent=pack, loc=<-1.2242 -9.5335 -3.1616>, rot=<-0.3667 -0.7909 -0.0499 0.4874>}
 dyn_display_01 {parent=dyn_display_MSTR, loc=<0 0 0>, rot=<0 0 0 1>}
 dyn_camera_MSTR {parent=pack, loc=<3.2961 -7.5352 -2.0515>, rot=<0.7051 -0.0187 0.6757 -0.2144>}
 dyn_camera_01 {parent=dyn_camera_MSTR, loc=<0 0 0>, rot=<0 0 0 1>}
 dyn_camera_02 {parent=dyn_camera_01, loc=<9.5805 0 0>, rot=<0 0 -0.7071 -0.7071>}
 camera_lens {parent=dyn_camera_02, loc=<7.9143 -2.4268 0.0692>, rot=<-0.0098 -0.0006 -0.0427 -0.999>}
 dyn_cord_cam_l_MSTR {parent=dyn_camera_02, loc=<7.4007 1.0472 2.3614>, rot=<0.3308 0.482 -0.5414 -0.6043>}
 dyn_cord_cam_l_01 {parent=dyn_cord_cam_l_MSTR, loc=<0 0 0>, rot=<0 0 0 1>}
 pelvis_ref {parent=pack, loc=<0 17.207 -40.1139>, rot=<0.0127 0.707 -0.0127 -0.707>}
 spine_01_ref {parent=pelvis_ref, loc=<6.7494 -0.6388 0>, rot=<0 0 0.034 -0.9994>}
 spine_02_ref {parent=spine_01_ref, loc=<7.2518 -0.035 0>, rot=<0 0 0.0238 -0.9997>}
 spine_03_ref {parent=spine_02_ref, loc=<6.0104 0 0>, rot=<0 0 0.0149 -0.9999>}
 spine_04_ref {parent=spine_03_ref, loc=<6.1005 0 0>, rot=<0 0 0 1>}
 spine_05_ref {parent=spine_04_ref, loc=<5.9951 0 0>, rot=<0 0 0 1>}
 cord_mid {parent=pack, loc=<4.514 -11.5326 2.8103>, rot=<-0.1469 0.0552 -0.1623 -0.9742>}
]
