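# Physics asset for the PawJasmineGravity backpack cosmetic, giving its dangling and attached parts bodies for secondary motion.
# Two capsules form the pack body, with chains for backpack sides, camera, display, right-side cords, trinket, straps and zippers.
# Each chain starts at a tiny MSTR capsule with free swing and twist against the pack.
# Child joints carry per-part angular limits, generally tight on camera and display, looser on cords and zippers.
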PhysicsAsset /BRCosmetics/Accessories/FORT_Backpacks/Backpack_M_MED_PawJasmineGravity/Meshes/M_MED_PawJasmineGravity_Pack_Physics
bodies {
 dyn_backpack_l_01 [{shapeType=Capsule, length=4, radius=1, shapeName=dyn_backpack_l_01_capsule, center=<4 0 0>}]
 dyn_backpack_l_MSTR [{shapeType=Capsule, length=0, radius=0.5, shapeName=dyn_backpack_l_MSTR_capsule}]
 dyn_backpack_r_01 [{shapeType=Capsule, length=4, radius=1, shapeName=dyn_backpack_r_01_capsule, center=<3 0 0>}]
 dyn_backpack_r_MSTR [{shapeType=Capsule, length=0, radius=0.5, shapeName=dyn_backpack_r_MSTR_capsule}]
 dyn_camera_01 [{shapeType=Capsule, length=9.6828, radius=1.7266, shapeName=dyn_camera_01_capsule, center=<5.8311 0 0>, rotation=<rot 90 0 0>}]
 dyn_camera_02 [{shapeType=Capsule, length=6.0603, radius=4.0801, shapeName=dyn_camera_02_capsule, center=<2.6696 -3.2501 0>, rotation=<rot 90 0 0>}]
 dyn_camera_MSTR [{shapeType=Capsule, length=0, radius=0.5, shapeName=dyn_camera_MSTR_capsule}]
 dyn_cord_r_b_01 [{shapeType=Capsule, length=0, radius=1, shapeName=dyn_cord_r_b_01_capsule, center=<1.5 0 0>}]
 dyn_cord_r_b_02 [{shapeType=Capsule, length=0.7559, radius=0.6412, shapeName=dyn_cord_r_b_02_capsule, center=<1.5 0 0>, rotation=<rot 0.0001 0 82.875>}]
 dyn_cord_r_b_MSTR [{shapeType=Capsule, length=0, radius=0.5, shapeName=dyn_cord_r_b_MSTR_capsule}]
 dyn_cord_r_f_01 [{
   shapeType=Capsule
   length=1.3661
   radius=1
   shapeName=dyn_cord_r_f_01_capsule
   center=<2.9522 -0.0108 -0.0051>
   rotation=<rot 0.0001 0 64.6464>
  }]
 dyn_cord_r_f_02 [{shapeType=Capsule, length=1.3473, radius=0.6287, shapeName=dyn_cord_r_f_02_capsule, center=<2.5163 0 0>, rotation=<rot 0 0 62.7607>}]
 dyn_cord_r_f_03 [{
   shapeType=Capsule
   length=0
   radius=1.4506
   shapeName=dyn_cord_r_f_03_capsule
   center=<1.5149 -0.2099 0.2683>
   rotation=<rot 0.0002 0 51.9651>
  }]
 dyn_cord_r_f_MSTR [{shapeType=Capsule, length=0, radius=0.5, shapeName=dyn_cord_r_f_MSTR_capsule}]
 dyn_display_01 [{shapeType=Capsule, length=4.3296, radius=4.0815, shapeName=dyn_display_01_capsule, center=<5.2146 0 0>}]
 dyn_display_MSTR [{shapeType=Capsule, length=0, radius=0.5, shapeName=dyn_display_MSTR_capsule}]
 dyn_dyn_trinket_01 [{shapeType=Capsule, length=0.8807, radius=0.77, shapeName=dyn_dyn_trinket_01_capsule, center=<0.9937 0 0>, rotation=<rot 90 0 0>}]
 dyn_dyn_trinket_02 [{shapeType=Capsule, length=0.6005, radius=1, shapeName=dyn_dyn_trinket_02_capsule, center=<2 0 0>}]
 dyn_dyn_trinket_03 [
  {shapeType=Sphere, radius=1, shapeName=dyn_dyn_trinket_03_sphere, center=<3.9759 0 -3.0443>}
  {
   shapeType=Capsule
   length=4.1979
   radius=1.7018
   shapeName=dyn_dyn_trinket_03_capsule
   center=<5.4161 0 -0.417>
   rotation=<rot 81.5483 0.0001 0>
  }
 ]
 dyn_dyn_trinket_MSTR [{shapeType=Capsule, length=0, radius=0.5, shapeName=dyn_dyn_trinket_MSTR_capsule}]
 dyn_strap_l_01 [{shapeType=Capsule, length=1.7869, radius=0.4, shapeName=dyn_strap_l_01_capsule, center=<2 0 0>}]
 dyn_strap_l_02 [{shapeType=Capsule, length=1.7869, radius=0.4, shapeName=dyn_strap_l_02_capsule, center=<2 0 0>}]
 dyn_strap_l_MSTR [{shapeType=Capsule, length=0, radius=0.5, shapeName=dyn_strap_l_MSTR_capsule}]
 dyn_strap_r_01 [{shapeType=Capsule, length=1.7869, radius=0.4, shapeName=dyn_strap_r_01_capsule, center=<2 0 0>}]
 dyn_strap_r_02 [{shapeType=Capsule, length=1.7869, radius=0.4, shapeName=dyn_strap_r_02_capsule, center=<2 0 0>}]
 dyn_strap_r_MSTR [{shapeType=Capsule, length=0, radius=0.5, shapeName=dyn_strap_r_MSTR_capsule}]
 dyn_zipper_bot_01 [{shapeType=Capsule, length=1.6426, radius=0.6358, shapeName=dyn_zipper_bot_01_capsule, center=<0.9367 0 0>, rotation=<rot 90 0 0>}]
 dyn_zipper_bot_MSTR [{shapeType=Capsule, length=0, radius=0.5, shapeName=dyn_zipper_bot_MSTR_capsule}]
 dyn_zipper_top_01 [{shapeType=Capsule, length=1.47, radius=1, shapeName=dyn_zipper_top_01_capsule, center=<0.983 0 0>, rotation=<rot 90 0 0>}]
 dyn_zipper_top_MSTR [{shapeType=Capsule, length=0, radius=0.5, shapeName=dyn_zipper_top_MSTR_capsule}]
 pack [
  {shapeType=Capsule, length=14.7215, radius=10.327, shapeName=pack_capsule, center=<0 -6.6168 -9.1616>}
  {shapeType=Capsule, length=8.0416, radius=4.9282, shapeName=pack_capsule, center=<-10.4284 -8.2843 -16.4063>}
 ]
}
constraints [
 {
  bone1Name=dyn_backpack_l_MSTR
  bone2Name=pack
  swing1LimitDegrees=45
  swing1Motion=Free
  swing2LimitDegrees=45
  swing2Motion=Free
  twistLimitDegrees=45
  twistMotion=Free
 }
 {bone1Name=dyn_backpack_l_01, bone2Name=dyn_backpack_l_MSTR, swing1LimitDegrees=16, swing2LimitDegrees=4, twistLimitDegrees=4}
 {
  bone1Name=dyn_backpack_r_MSTR
  bone2Name=pack
  swing1LimitDegrees=45
  swing1Motion=Free
  swing2LimitDegrees=45
  swing2Motion=Free
  twistLimitDegrees=45
  twistMotion=Free
 }
 {bone1Name=dyn_backpack_r_01, bone2Name=dyn_backpack_r_MSTR, swing1LimitDegrees=14, swing2LimitDegrees=4, twistLimitDegrees=4}
 {
  bone1Name=dyn_camera_MSTR
  bone2Name=pack
  swing1LimitDegrees=45
  swing1Motion=Free
  swing2LimitDegrees=45
  swing2Motion=Free
  twistLimitDegrees=45
  twistMotion=Free
 }
 {bone1Name=dyn_camera_01, bone2Name=dyn_camera_MSTR, swing1LimitDegrees=11, swing2LimitDegrees=6, twistLimitDegrees=1}
 {bone1Name=dyn_camera_02, bone2Name=dyn_camera_01, swing1LimitDegrees=15, swing2LimitDegrees=0.1, twistLimitDegrees=0.1}
 {
  bone1Name=dyn_cord_r_b_MSTR
  bone2Name=pack
  swing1LimitDegrees=45
  swing1Motion=Free
  swing2LimitDegrees=45
  swing2Motion=Free
  twistLimitDegrees=45
  twistMotion=Free
 }
 {bone1Name=dyn_cord_r_b_01, bone2Name=dyn_cord_r_b_MSTR, swing1LimitDegrees=50, swing2LimitDegrees=15, twistLimitDegrees=10}
 {bone1Name=dyn_cord_r_b_02, bone2Name=dyn_cord_r_b_01, swing1LimitDegrees=50, swing2LimitDegrees=50, twistLimitDegrees=10}
 {
  bone1Name=dyn_cord_r_f_MSTR
  bone2Name=pack
  swing1LimitDegrees=45
  swing1Motion=Free
  swing2LimitDegrees=45
  swing2Motion=Free
  twistLimitDegrees=45
  twistMotion=Free
 }
 {bone1Name=dyn_cord_r_f_01, bone2Name=dyn_cord_r_f_MSTR, swing1LimitDegrees=45, swing2LimitDegrees=15, twistLimitDegrees=4}
 {bone1Name=dyn_cord_r_f_02, bone2Name=dyn_cord_r_f_01, swing1LimitDegrees=45, swing2LimitDegrees=45, twistLimitDegrees=8}
 {bone1Name=dyn_cord_r_f_03, bone2Name=dyn_cord_r_f_02, swing1LimitDegrees=45, swing2LimitDegrees=45, twistLimitDegrees=8}
 {
  bone1Name=dyn_display_MSTR
  bone2Name=pack
  swing1LimitDegrees=45
  swing1Motion=Free
  swing2LimitDegrees=45
  swing2Motion=Free
  twistLimitDegrees=45
  twistMotion=Free
 }
 {bone1Name=dyn_display_01, bone2Name=dyn_display_MSTR, swing1LimitDegrees=1, swing2LimitDegrees=10, twistLimitDegrees=3}
 {
  bone1Name=dyn_dyn_trinket_MSTR
  bone2Name=pack
  swing1LimitDegrees=45
  swing1Motion=Free
  swing2LimitDegrees=45
  swing2Motion=Free
  twistLimitDegrees=45
  twistMotion=Free
 }
 {bone1Name=dyn_dyn_trinket_01, bone2Name=dyn_dyn_trinket_MSTR, swing1LimitDegrees=60, swing2LimitDegrees=10, twistLimitDegrees=2}
 {bone1Name=dyn_dyn_trinket_02, bone2Name=dyn_dyn_trinket_01, swing1LimitDegrees=35, swing2LimitDegrees=15, twistLimitDegrees=5}
 {bone1Name=dyn_dyn_trinket_03, bone2Name=dyn_dyn_trinket_02, swing1LimitDegrees=15, swing2LimitDegrees=30, twistLimitDegrees=10}
 {
  bone1Name=dyn_strap_l_MSTR
  bone2Name=pack
  swing1LimitDegrees=45
  swing1Motion=Free
  swing2LimitDegrees=45
  swing2Motion=Free
  twistLimitDegrees=45
  twistMotion=Free
 }
 {bone1Name=dyn_strap_l_01, bone2Name=dyn_strap_l_MSTR, swing1LimitDegrees=35, swing2LimitDegrees=4, twistLimitDegrees=2}
 {bone1Name=dyn_strap_l_02, bone2Name=dyn_strap_l_01, swing1LimitDegrees=45, swing2LimitDegrees=25, twistLimitDegrees=15}
 {
  bone1Name=dyn_strap_r_MSTR
  bone2Name=pack
  swing1LimitDegrees=45
  swing1Motion=Free
  swing2LimitDegrees=45
  swing2Motion=Free
  twistLimitDegrees=45
  twistMotion=Free
 }
 {bone1Name=dyn_strap_r_01, bone2Name=dyn_strap_r_MSTR, swing1LimitDegrees=32, swing2LimitDegrees=3, twistLimitDegrees=2}
 {bone1Name=dyn_strap_r_02, bone2Name=dyn_strap_r_01, swing1LimitDegrees=42, swing2LimitDegrees=20, twistLimitDegrees=10}
 {
  bone1Name=dyn_zipper_bot_MSTR
  bone2Name=pack
  swing1LimitDegrees=45
  swing1Motion=Free
  swing2LimitDegrees=45
  swing2Motion=Free
  twistLimitDegrees=45
  twistMotion=Free
 }
 {bone1Name=dyn_zipper_bot_01, bone2Name=dyn_zipper_bot_MSTR, swing1LimitDegrees=80, swing2LimitDegrees=10, twistLimitDegrees=20}
 {
  bone1Name=dyn_zipper_top_MSTR
  bone2Name=pack
  swing1LimitDegrees=45
  swing1Motion=Free
  swing2LimitDegrees=45
  swing2Motion=Free
  twistLimitDegrees=45
  twistMotion=Free
 }
 {bone1Name=dyn_zipper_top_01, bone2Name=dyn_zipper_top_MSTR, swing1LimitDegrees=90, swing2LimitDegrees=2, twistLimitDegrees=2}
]
