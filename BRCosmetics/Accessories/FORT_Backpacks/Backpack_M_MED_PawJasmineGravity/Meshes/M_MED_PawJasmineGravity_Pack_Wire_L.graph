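# Physics asset for the left-side wire of the PawJasmineGravity backpack cosmetic, defining collision capsules and joint limits for its dangling parts.
# A two-capsule pack body is the hub for a short backpack bone, a two-segment cord chain and a one-segment cord_cam chain.
# Each cord chain joins the pack through a small MSTR body with free swing and twist, while the other joints have angular limits.
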PhysicsAsset /BRCosmetics/Accessories/FORT_Backpacks/Backpack_M_MED_PawJasmineGravity/Meshes/M_MED_PawJasmineGravity_Pack_Wire_L
bodies {
 dyn_backpack_l_01 [{
   shapeType=Capsule
   length=3.7073
   radius=1.7387
   shapeName=dyn_backpack_l_01_capsule
   center=<2.3561 -1.1399 1.3115>
   rotation=<rot 87.585 150.4567 126.7468>
  }]
 dyn_cord_cam_l_01 [{
   shapeType=Capsule
   length=7.2523
   radius=1.2714
   shapeName=dyn_cord_cam_l_01_capsule
   center=<3.7313 -0.5865 0.0839>
   rotation=<rot -0.9978 89.2852 -83.3226>
  }]
 dyn_cord_cam_l_MSTR [{shapeType=Capsule, length=0.505, radius=0.505, shapeName=dyn_cord_cam_l_MSTR_capsule, rotation=<rot 0 -90 0>}]
 dyn_cord_l_01 [{shapeType=Capsule, length=3.9389, radius=0.5, shapeName=dyn_cord_l_01_capsule, center=<3.3827 0 0.1064>}]
 dyn_cord_l_02 [{
   shapeType=Capsule
   length=1.6073
   radius=0.7174
   shapeName=dyn_cord_l_02_capsule
   center=<2.4913 0.4504 0.0979>
   rotation=<rot 0.0001 0 16.0123>
  }]
 dyn_cord_l_MSTR [{shapeType=Capsule, length=0, radius=0.5, shapeName=dyn_cord_l_MSTR_capsule}]
 pack [
  {shapeType=Capsule, length=15.0605, radius=10.1156, shapeName=pack_capsule, center=<0 -7.7771 -9.1616>}
  {shapeType=Capsule, length=8.0416, radius=4.9282, shapeName=pack_capsule, center=<-10.4284 -8.2843 -16.4063>}
 ]
}
constraints [
 {bone1Name=dyn_backpack_l_01, bone2Name=pack, swing1LimitDegrees=16, swing2LimitDegrees=4, twistLimitDegrees=4}
 {
  bone1Name=dyn_cord_l_MSTR
  bone2Name=pack
  swing1LimitDegrees=45
  swing1Motion=Free
  swing2LimitDegrees=45
  swing2Motion=Free
  twistLimitDegrees=45
  twistMotion=Free
 }
 {bone1Name=dyn_cord_l_01, bone2Name=dyn_cord_l_MSTR, swing1LimitDegrees=50, swing2LimitDegrees=15, twistLimitDegrees=10}
 {bone1Name=dyn_cord_l_02, bone2Name=dyn_cord_l_01, swing1LimitDegrees=25, swing2LimitDegrees=6, twistLimitDegrees=10}
 {
  bone1Name=dyn_cord_cam_l_MSTR
  bone2Name=pack
  swing1LimitDegrees=45
  swing1Motion=Free
  swing2LimitDegrees=45
  swing2Motion=Free
  twistLimitDegrees=45
  twistMotion=Free
 }
 {bone1Name=dyn_cord_cam_l_01, bone2Name=dyn_cord_cam_l_MSTR, swing1LimitDegrees=55, swing2LimitDegrees=45, twistLimitDegrees=1}
]
